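; Bug-biome porta potty hiding prop whose interaction has a rare random chance to spawn a bug swarm instead of running normal behavior.
; SpawnBug starts a living-world encounter on authority through a child-actor thrower point provider and binds last-actor death to ending it; remotes play the door swing.
; On hidden travel completion, the server applies a configured gameplay effect to the pawn.
; Door swings also spawn an attached enter/exit Niagara effect.
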(event EventBeginPlay
  (|CallParentFunction) ; This node is disabled and will not be called.\nDrag off pins to build functionality.
)

(event Collision|EventActorBeginOverlap
  (params (OtherActor :type Actor%20Object%20Reference))
  (|CallParentFunction OtherActor) ; This node is disabled and will not be called.\nDrag off pins to build functionality.
)

(event EventTick
  (params (DeltaSeconds :type Float%20%28single-precision%29)))

(event EventHiddenTravelCompleteEvent
  (params (TargetHidingProp :type B%20Hiding%20Prop%20Object%20Reference) (Pawn :type Fort%20Pawn%20Object%20Reference))
  (bind _self self)
  (bind _returnvalue (Ability|GetFortAbilitySystemComponent _self))
  (|CallParentFunction TargetHidingProp Pawn)
  (Networking|SwitchHasAuthority
    (:Authority
      (Utilities|IsValid Pawn
        (:Is%20Valid
          (Utilities|IsValid _returnvalue
            (:Is%20Valid
              (GameplayEffects|FortApplyGameplayEffectToTarget :self _returnvalue :GameplayEffectClass (Variables|Default|GetGameplay%20Effect%20Class) :TargetOfEffect (Ability|GetFortAbilitySystemComponent Pawn) :OptionalSourceObject _self :OptionalEffectCauser _self :OptionalDamageSource _self :Level 0.0))))))))

(event Interaction|EventBlueprintOnInteract
  (params (InteractingPawn :type Fort%20Pawn%20Object%20Reference) (InteractionBeingAttempted :type EInteractionBeingAttempted%20Enum :default "FirstInteraction"))
  (if (Math|Random|RandomBoolwithWeight 0.99)
    (|CallParentFunction InteractingPawn InteractionBeingAttempted)
    (else
      (CallFunction|SpawnBug))))

(event Custom|EndEncounter
  (params (EncounterInstance :type Fort%20Athena%20Living%20World%20Encounter%20Instance%20Object%20Reference))
  (Class|BPFLLagerCommunFunctions|EndActivityEncounter (Game|GetGameState) (stale Custom|SpawnBug _returnvalue)))

(event Custom|SpawnBug
  (Networking|SwitchHasAuthority
    (:Authority
      (bind (_isvalid _returnvalue) (Default|StartActivityEncounterStaticPointProvider (Game|GetGameState) self "/Lager/Lager/Encounters/S37/LagerEncounter_BugPorto_1Swarm.LagerEncounter_BugPorto_1Swarm" (Utilities|Array|MakeArray (Utilities|Casting|CastToBP_StaticPointProvider_Thrower (Variables|/Script/Engine.ChildActorComponent|GetChildActor (Variables|Default|GetChildActor))))))
      (EventDispatchers|BindEventtoOnLastActorDied _returnvalue (delegate Custom|EndEncounter)))
    (:Remote
      (CallFunction|DoorSwingAnim))))

(event EventDoorSwingAnim
  (|CallParentFunction)
  (Niagara|SpawnSystemAttached "/BugBiomeGameplay/Assets/VFX/NS_BugPortaPotty_EnterExit.NS_BugPortaPotty_EnterExit" (Variables|Debug|GetStaticMeshComponent)))

(fn ConstructionScript
  (|CallParentFunction))
