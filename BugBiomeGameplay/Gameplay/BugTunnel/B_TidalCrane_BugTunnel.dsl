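; Bug tunnel Blueprint for the Tidal Crane, an interactable that shows a "Travel" prompt and extends a parent class's player-launch behavior.
; When players launch off the top, it runs the parent logic, then on the server applies the goo gameplay effect to the pawn through the tunnel's own ability system component.
; BeginPlay, overlap and Tick hold no active logic, and the construction script only calls the parent.
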(event EventBeginPlay
  (|CallParentFunction) ; This node is disabled and will not be called.\nDrag off pins to build functionality.
)

(event Collision|EventActorBeginOverlap
  (params (OtherActor :type Actor%20Object%20Reference))
  (|CallParentFunction OtherActor) ; This node is disabled and will not be called.\nDrag off pins to build functionality.
)

(event EventTick
  (params (DeltaSeconds :type Float%20%28single-precision%29)))

(event EventLaunchPlayersOffTop
  (params (PlayerPawn :type Fort%20Player%20Pawn%20Object%20Reference))
  (bind _self self)
  (bind _returnvalue (Ability|GetFortAbilitySystemComponent _self))
  (|CallParentFunction PlayerPawn)
  (Networking|SwitchHasAuthority
    (:Authority
      (Utilities|IsValid PlayerPawn
        (:Is%20Valid
          (Utilities|IsValid _returnvalue
            (:Is%20Valid
              (GameplayEffects|FortApplyGameplayEffectToTarget :self _returnvalue :GameplayEffectClass (Variables|Default|GetGooEffectClass) :TargetOfEffect (Ability|GetFortAbilitySystemComponent PlayerPawn) :OptionalSourceObject _self :OptionalEffectCauser _self :OptionalDamageSource _self :Level 0.0))))))))

(fn BlueprintGetInteractionString
  (params (InteractingPawn :type Fort%20Pawn%20Object%20Reference) (InteractionBeingAttempted :type EInteractionBeingAttempted%20Enum :default "FirstInteraction"))
  (Utilities|IsValid InteractingPawn
    (:Is%20Valid
      (return "NSLOCTEXT(\"[BC51AFD3AC19CEB3A55682F15D8244FC]\", \"5D63319C4762A11E8514179A0EDF7415\", \"Travel\")"))))

(fn ConstructionScript
  (|CallParentFunction))
